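# Post-process style Anim Blueprint for the F_MED_Meteorwomen_Alt cosmetic body, copying the attached parent's pose and adding secondary motion.
# A Control Rig receives skydiving alpha and a helmet flag from the OptionB style tag.
# Constraints weight pouch master bones across spine and pelvis, then AnimDynamics springs swing zippers and pouches.
# Jaw, neck-suit and head-mesh blend nodes sit outside the output chain, leaving the computed head-tilt alpha without visible effect.
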AnimBlueprint /BRCosmetics/Characters/Player/Female/Medium/Bodies/F_MED_Meteorwomen_Alt/Meshes/F_MED_Meteorwomen_Alt_AnimBP
target_skeleton=&/F_MED_Meteorwomen_Alt_Skeleton
graphs [
 AnimGraph {nodes [
   Layered blend per bone(BlendPoses_0=Copy Pose From Mesh(SourceMeshComponent=Get Body_Mesh()) {bCopyCurves=true, bUseAttachedParent=true}) {layerSetup [{branchFilters [{boneName=C_jaw}]}]}
   "Transform (Modify) Bone - Bone: C_jaw"(Alpha=1, Translation="0.000,10.644,0.000", Rotation="0,0,0", Scale="1,1,1") {translationMode=BMM_Additive, alphaBoolBlend {blendOption=Linear}, boneToModify {boneName=C_jaw}, translation {y=10.644}}
   n1 = "Transform (Modify) Bone - Bone: dyn_necksuit"(Alpha=0, Translation="0,-1,-1.5", Rotation="0,0,0", Scale="1,1,1") {translationMode=BMM_Additive, alphaBoolBlend {blendOption=Linear}, boneToModify {boneName=dyn_necksuit}, translation {y=-1, z=-1.5}}
   "Transform (Modify) Bone - Bone: dyn_necksuit"(ComponentPose=@n1, Alpha=Get Head Tilt Alpha() -> (Head Tilt Alpha:double), Translation="0,4,3", Rotation="0,0,0", Scale="1,1,1") {translationMode=BMM_Additive, alphaBoolBlend {blendOption=Linear}, boneToModify {boneName=dyn_necksuit}, translation {y=4, z=3}}
   n2 = Control Rig!(
    Source=Copy Pose From Mesh() {bCopyCurves=true, bUseAttachedParent=true}
    Alpha
    SkyDivingAlphaControl:double=Get IsSkyDivingAlpha() -> (IsSkyDivingAlpha:double)
    Helmetctrlrig:bool=Get Helmet?() -> (Helmet?:bool)
   ) -> (Pose) {
    alphaBoolBlend {blendOption=Linear}
    controlRigAssetReference {blueprintRigClass=&/F_MED_Meteorwomen_Alt_CtrlRig.F_MED_Meteorwomen_Alt_CtrlRig_C}
   }
   n3 = "Transform (Modify) Bone - Bone: attach_cape"(ComponentPose=Local To Component(LocalPose=@n2), Alpha=0, Translation="1,-1,-1", Rotation="0,0,0", Scale="1,1,1") {translationMode=BMM_Additive, alphaBoolBlend {blendOption=Linear}, boneToModify {boneName=attach_cape}, translation=<1 -1 -1>}
   n4 = "Constraint - dyn_mstr_pouch_r (PPP)"!(ComponentPose=@n3, Alpha=1, ConstraintWeights_0=0.0671, ConstraintWeights_1:float=0.0678, ConstraintWeights_2:float=0.8651) -> (Pose) {
    alphaBoolBlend {blendOption=Linear}
    boneToModify {boneName=dyn_mstr_pouch_r}
    constraintSetup [{targetBone {boneName=spine_02}}, {targetBone {boneName=spine_01}}, {targetBone {boneName=pelvis}}]
    constraintWeights [1,1,1]
   }
   n5 = "Constraint - dyn_mstr_pouch_l (PPP)"!(ComponentPose=@n4, Alpha=1, ConstraintWeights_0=0.0671, ConstraintWeights_1:float=0.0678, ConstraintWeights_2:float=0.8651) -> (Pose) {
    alphaBoolBlend {blendOption=Linear}
    boneToModify {boneName=dyn_mstr_pouch_l}
    constraintSetup [{targetBone {boneName=spine_02}}, {targetBone {boneName=spine_01}}, {targetBone {boneName=pelvis}}]
    constraintWeights [1,1,1]
   }
   n6 = Anim Dynamics - Bone: dyn_zipper_l(ComponentPose=@n5, Alpha=1, ExternalForce="0,0,0") {
    angularSpringConstant=0.9
    bAngularSpring=true
    alphaBoolBlend {blendOption=Linear}
    boundBone {boneName=dyn_zipper_l}
    componentAppliedLinearAccClamp=<3000 3000 30000>
    componentLinearAccScale=<0.5 0.5 0.5>
    componentLinearVelScale=<0.5 0.5 0.5>
    physicsBodyDefinitions [{
      boundBone {boneName=dyn_zipper_l}
      boxExtents=<10 10 10>
      constraintSetup {angularLimitsMax=<30 0 60>, angularLimitsMin=<-30 -60 -60>, angularTarget=<0 0 0>}
      localJointOffset=<2 0 0>
     }]
    relativeSpaceBone {boneName=pelvis}
   }
   n7 = Anim Dynamics - Bone: dyn_zipper_r(ComponentPose=@n6, Alpha=1, ExternalForce="0,0,0") {
    angularSpringConstant=0.9
    bAngularSpring=true
    alphaBoolBlend {blendOption=Linear}
    boundBone {boneName=dyn_zipper_r}
    componentAppliedLinearAccClamp=<3000 3000 30000>
    componentLinearAccScale=<0.5 0.5 0.5>
    componentLinearVelScale=<0.5 0.5 0.5>
    physicsBodyDefinitions [{
      boundBone {boneName=dyn_zipper_r}
      boxExtents=<10 10 10>
      constraintSetup {angularLimitsMax=<30 0 60>, angularLimitsMin=<-30 -60 -60>, angularTarget=<0 0 0>}
      localJointOffset=<2 0 0>
     }]
    relativeSpaceBone {boneName=pelvis}
   }
   n8 = Anim Dynamics - Bone: dyn_pouch_l(ComponentPose=@n7, Alpha=0.8, ExternalForce="0,0,0") {
    angularSpringConstant=0.9
    bLinearSpring=true
    linearSpringConstant=100
    simulationSpace=RootRelative
    alphaBoolBlend {blendOption=Linear}
    boundBone {boneName=dyn_pouch_l}
    componentAppliedLinearAccClamp=<3000 3000 30000>
    componentLinearAccScale=<0.3 0.25 0.25>
    componentLinearVelScale=<0.3 0.25 0.25>
    physicsBodyDefinitions [{
      boundBone {boneName=dyn_pouch_l}
      boxExtents=<20 30 20>
      constraintSetup {angularLimitsMax=<50 0 10>, angularLimitsMin=<-50 -30 -10>, angularTarget=<0 0 0>}
      localJointOffset=<2 0 -3>
     }]
    relativeSpaceBone {boneName=pelvis}
   }
   n9 = Anim Dynamics - Bone: dyn_pouch_r(ComponentPose=@n8, Alpha=0.8, ExternalForce="0,0,0") {
    angularSpringConstant=0.9
    bLinearSpring=true
    linearSpringConstant=100
    simulationSpace=RootRelative
    alphaBoolBlend {blendOption=Linear}
    boundBone {boneName=dyn_pouch_r}
    componentAppliedLinearAccClamp=<3000 3000 30000>
    componentLinearAccScale=<0.3 0.25 0.25>
    componentLinearVelScale=<0.3 0.25 0.25>
    physicsBodyDefinitions [{
      boundBone {boneName=dyn_pouch_r}
      boxExtents=<20 30 20>
      constraintSetup {angularLimitsMax=<50 0 10>, angularLimitsMin=<-50 -30 -10>, angularTarget=<0 0 0>}
      localJointOffset=<2 0 -3>
     }]
    relativeSpaceBone {boneName=pelvis}
   }
   Output Pose(Result=Component To Local(ComponentPose=@n9)) {layerGroup=None}
   n10 = Layered blend per bone(
    BasePose=Copy Pose From Mesh() {bCopyCurves=true, bUseAttachedParent=true}
    BlendPoses_0=Copy Pose From Mesh(SourceMeshComponent=Get Head_Mesh() -> (Head_Mesh:Skeletal Mesh Component Object Reference))
   ) {layerSetup [{branchFilters [{boneName=thigh_l}, {boneName=thigh_r}, {boneName=calf_l}, {boneName=calf_r}]}]}
   Save cached pose 'Start_Cache'(Pose=@n10) {cacheName=Start_Cache}
   Use cached pose 'Start_Cache'()
   Get Body_Mesh()
  ]}
 EventGraph {nodes [
   Event BlueprintUpdateAnimation()
   > Set Helmet?(Helmet?:bool=CharacterItemDefinitionHasAnyMetaTag(Tags=Make Array("[0]"=(TagName="Cosmetics.MetaTags.Style.OptionB")))) -> (Output_Get:bool)
   n1 = Get bIsSkydiveFloating()
   n2 = Get bIsSkydiveDiveMode()
   n3 = Get bIsSkyDiving()
   > Set IsSkyDivingAlpha(IsSkyDivingAlpha:double=SelectFloat(A=1, B=0, bPickA=OR Boolean!(A=@n2, B=@n1, C:bool=@n3) -> (ReturnValue))) -> (Output_Get:double)
   > Branch(Condition=Get Helmet?())
   n4 = BreakTransform(InTransform=GetDeltaTransformFromRefPose(self=GetOwnerSkeletalMeshComponent(), BoneName=head, BaseName=spine_05))
   > Set Head Tilt Alpha(Head Tilt Alpha:double=MapRangeClamped(Value=BreakRotator(InRot=@n4.Rotation).Yaw, InRangeA=0, InRangeB=15)) -> (Output_Get:double)
   Set Body_Mesh(Body_Mesh=GetOwnerSkeletalMeshComponent())
   > n5 = Cast To PlayerPawn_Athena(Object=TryGetPawnOwner())
   > n6 = Set Pawn Owner(Pawn Owner:Player Pawn Athena Object Reference=@n5) -> (Output_Get:Player Pawn Athena Object Reference)
   > Set Head_Mesh(Head_Mesh:Skeletal Mesh Component Object Reference=GetSkeletalMeshForPartType(self=@n6, PartType=Head)) -> (Output_Get:Skeletal Mesh Component Object Reference)
   GetSkeletalMeshForPartType(PartType=Head)
   GetMetadata!(SubjectFrameHandle:"Subject Frame Handle Structure (by ref)") -> (Metadata:Subject Metadata Structure)
   Event BlueprintInitializeAnimation()
  ]}
]
variables [
 Body_Mesh {type=SkeletalMeshComponent}
 Head Tilt Alpha {type=number}
 Head_Mesh {type=SkeletalMeshComponent}
 Helmet? {type=boolean}
 IsSkyDivingAlpha {type=number}
 Pawn Owner {type=PlayerPawn_Athena_C}
]
