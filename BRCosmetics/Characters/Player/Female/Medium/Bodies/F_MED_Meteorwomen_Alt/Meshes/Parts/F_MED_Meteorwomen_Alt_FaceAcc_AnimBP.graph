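# Animation Blueprint for the Meteorwomen Alt female face accessory, which follows the parent body and orients its tube bones.
# Copy Pose From Mesh feeds a Control Rig that receives an is-in-front-end flag.
# Two Look At nodes then aim the tube's body and head base bones at separate tube bones.
# Their alpha is remapped each update from head yaw relative to spine_05, compared with the reference pose, for one turn direction.
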AnimBlueprint /BRCosmetics/Characters/Player/Female/Medium/Bodies/F_MED_Meteorwomen_Alt/Meshes/Parts/F_MED_Meteorwomen_Alt_FaceAcc_AnimBP
target_skeleton=&/F_MED_Meteorwomen_Alt_FaceAcc_Skeleton
graphs [
 AnimGraph {nodes [
   n1 = Control Rig!(Source=Copy Pose From Mesh() {bUseAttachedParent=true}, Alpha, FrontEndctrl:bool=Get bIsInFrontEnd()) -> (Pose) {
    alphaBoolBlend {blendOption=Linear}
    controlRigAssetReference {blueprintRigClass=&/F_MED_Meteorwomen_Alt_FaceAcc_CtrlRig.F_MED_Meteorwomen_Alt_FaceAcc_CtrlRig_C}
   }
   n2 = Get TubeLookAlpha() -> (TubeLookAlpha:double)
   n3 = Look At - Bone: dyn_tube_bodybase(ComponentPose=Local To Component(LocalPose=@n1), Alpha=@n2) {
    bUseLookUpAxis=true
    alphaBoolBlend {blendOption=Linear}
    boneToModify {boneName=dyn_tube_bodybase}
    lookAtTarget {boneReference {boneName=dyn_tube_01}}
    lookAt_Axis {axis=<1 0>}
   }
   n4 = Look At - Bone: dyn_tube_headbase(ComponentPose=@n3, Alpha=@n2) {
    bUseLookUpAxis=true
    alphaBoolBlend {blendOption=Linear}
    boneToModify {boneName=dyn_tube_headbase}
    lookAtTarget {boneReference {boneName=dyn_tube_02}}
    lookAt_Axis {axis=<1 0>}
   }
   Output Pose(Result=Component To Local(ComponentPose=@n4)) {layerGroup=None}
  ]}
 EventGraph {nodes [
   Event BlueprintUpdateAnimation()
   n1 = BreakTransform(InTransform=GetDeltaTransformFromRefPose(self=GetOwnerSkeletalMeshComponent(), BoneName=head, BaseName=spine_05))
   > Set TubeLookAlpha(TubeLookAlpha:double=MapRangeClamped(Value=BreakRotator(InRot=@n1.Rotation).Yaw, InRangeA=0, InRangeB=-70)) -> (Output_Get:double)
  ]}
]
variables [TubeLookAlpha {type=number}]
